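# Animation Blueprint for the DangerIndex pickaxe that adds procedural jiggle to its umbrella bones.
# The input pose runs through five chained Anim Dynamics nodes in component space: umbrela_main first, then umbrella_01 to umbrella_04, then back to local space.
# Each bone uses an angular spring with tight limits, high damping, gravity override and scaled component motion; the four numbered bones share near-identical setups.
# The EventGraph holds only unconnected stub nodes, with no variables or logic.
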AnimBlueprint /BRCosmetics/Weapons/FORT_Melee/Pickaxe_DangerIndex/Meshes/DangerIndex_Axe_AnimBP
target_skeleton=&/DangerIndex_Axe_Skeleton
graphs [
 AnimGraph {nodes [
   AnimGraphNode_SubInput_0 = Input Pose()
   n1 = Anim Dynamics - Bone: umbrela_main(ComponentPose=Local To Component(LocalPose=@AnimGraphNode_SubInput_0), Alpha=1, ExternalForce="0,0,0") {
    angularDampingOverride=0.9999
    bAngularSpring=true
    bOverrideAngularDamping=true
    bUseGravityOverride=true
    boundBone {boneName=umbrela_main}
    componentAppliedLinearAccClamp=<3000 3000 2500>
    componentLinearAccScale=<0.5 0.5 0.5>
    componentLinearVelScale=<0.5 0.5 0.5>
    gravityOverride {z=-980}
    physicsBodyDefinitions [{
      boundBone {boneName=umbrela_main}
      boxExtents=<20 20 20>
      constraintSetup {angularLimitsMax=<10 0 0>, angularLimitsMin=<-10 0 0>, angularTarget=<0 1 0>}
      localJointOffset=<0 1 0>
     }]
    simSpaceSettings {maxAngularAcceleration=35, maxAngularVelocity=7, simSpaceAngularAlpha=1}
   }
   n2 = Anim Dynamics - Bone: umbrella_01(ComponentPose=@n1, Alpha=1, ExternalForce="0,0,0") {
    angularDampingOverride=1
    angularSpringConstant=15000
    bAngularSpring=true
    bGravityOverrideInSimSpace=true
    bOverrideAngularDamping=true
    bUseGravityOverride=true
    boundBone {boneName=umbrella_01}
    componentAppliedLinearAccClamp=<3000 3000 2500>
    componentLinearAccScale=<0.25 0.25 0.25>
    componentLinearVelScale=<0.25 0.25 0.25>
    gravityOverride {z=-980}
    physicsBodyDefinitions [{
      boundBone {boneName=umbrella_01}
      boxExtents=<26 26 26>
      constraintSetup {angularLimitsMax=<2 0 5>, angularLimitsMin=<-2 0 -10>, angularTarget=<1 0 0>}
      localJointOffset=<0.1 0 0>
     }]
    simSpaceSettings {maxAngularAcceleration=35, maxAngularVelocity=7, simSpaceAngularAlpha=0.4}
   }
   n3 = Anim Dynamics - Bone: umbrella_02(ComponentPose=@n2, Alpha=1, ExternalForce="0,0,0") {
    angularDampingOverride=0.9995
    angularSpringConstant=15000
    bAngularSpring=true
    bGravityOverrideInSimSpace=true
    bOverrideAngularDamping=true
    bUseGravityOverride=true
    boundBone {boneName=umbrella_02}
    componentAppliedLinearAccClamp=<3000 3000 2500>
    componentLinearAccScale=<0.2 0.2 0.2>
    componentLinearVelScale=<0.2 0.2 0.2>
    gravityOverride {z=-980}
    physicsBodyDefinitions [{
      boundBone {boneName=umbrella_02}
      boxExtents=<26 26 26>
      constraintSetup {angularLimitsMax=<2 0 5>, angularLimitsMin=<-2 0 -10>, angularTarget=<1 0 0>}
      localJointOffset=<0.1 0 0>
     }]
    simSpaceSettings {maxAngularAcceleration=35, maxAngularVelocity=7, simSpaceAngularAlpha=0.4}
   }
   n4 = Anim Dynamics - Bone: umbrella_03(ComponentPose=@n3, Alpha=1, ExternalForce="0,0,0") {
    angularDampingOverride=1
    angularSpringConstant=15000
    bAngularSpring=true
    bGravityOverrideInSimSpace=true
    bOverrideAngularDamping=true
    bUseGravityOverride=true
    boundBone {boneName=umbrella_03}
    componentAppliedLinearAccClamp=<3000 3000 2500>
    componentLinearAccScale=<0.25 0.25 0.25>
    componentLinearVelScale=<0.25 0.25 0.25>
    gravityOverride {z=-980}
    physicsBodyDefinitions [{
      boundBone {boneName=umbrella_03}
      boxExtents=<26 26 26>
      constraintSetup {angularLimitsMax=<2 0 5>, angularLimitsMin=<-2 0 -10>, angularTarget=<1 0 0>}
      localJointOffset=<0.1 0 0>
     }]
    simSpaceSettings {maxAngularAcceleration=35, maxAngularVelocity=7, simSpaceAngularAlpha=0.4}
   }
   n5 = Anim Dynamics - Bone: umbrella_04(ComponentPose=@n4, Alpha=1, ExternalForce="0,0,0") {
    angularDampingOverride=0.9995
    angularSpringConstant=15000
    bAngularSpring=true
    bGravityOverrideInSimSpace=true
    bOverrideAngularDamping=true
    bUseGravityOverride=true
    boundBone {boneName=umbrella_04}
    componentAppliedLinearAccClamp=<3000 3000 2500>
    componentLinearAccScale=<0.2 0.2 0.2>
    componentLinearVelScale=<0.2 0.2 0.2>
    gravityOverride {z=-980}
    physicsBodyDefinitions [{
      boundBone {boneName=umbrella_04}
      boxExtents=<26 26 26>
      constraintSetup {angularLimitsMax=<2 0 5>, angularLimitsMin=<-2 0 -10>, angularTarget=<1 0 0>}
      localJointOffset=<0.1 0 0>
     }]
    simSpaceSettings {maxAngularAcceleration=35, maxAngularVelocity=7, simSpaceAngularAlpha=0.4}
   }
   Output Pose(Result=Component To Local(ComponentPose=@n5)) {layerGroup=None}
  ]}
 EventGraph {nodes [
   TryGetPawnOwner()
   Event BlueprintUpdateAnimation()
  ]}
]
variables []
